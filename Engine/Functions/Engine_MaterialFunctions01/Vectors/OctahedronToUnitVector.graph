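# Material function in the Vectors library category that unpacks a two-component octahedron-encoded value into a three-component unit vector.
# A Vector2 Octahedron input, with a preview default, feeds a Custom HLSL node.
# That node calls the shader-side OctahedronToUnitVector helper, so the decode math lives in shader code rather than in the graph, and its result is the Unit Vector output.
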MaterialFunction /Engine/Functions/Engine_MaterialFunctions01/Vectors/OctahedronToUnitVector
graph {
 properties {bExposeToLibrary=true, description=Unpacks a v2 Octahedron to a v3 vector., libraryCategoriesText [Vectors]}
 nodes [
  MaterialExpressionFunctionInput_2 = FunctionInput() {
   bUsePreviewValueAsDefault=true
   description=The V2 Octahedron to unpack to a V3 Vector.
   inputName=Octahedron
   inputType=FunctionInput_Vector2
   previewValue=<0 0 1 1>
  }
  n1 = Custom!(Oct=@MaterialExpressionFunctionInput_2) {
   code="return OctahedronToUnitVector(Oct);"
   description=Octahedron to Vector
   inputs [{
     inputName=Oct
     input {expression=&self:MaterialExpressionFunctionInput_2}
    }]
  }
  FunctionOutput(None=@n1) {outputName=Unit Vector}
 ]
}
