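# Refract is a library material function that computes a refracted direction vector for a ray crossing between two media.
# It divides the origin refractive index by the target index, with defaults for air and water, to get the index ratio.
# A Custom node passes the ray direction, surface normal and that ratio to the HLSL refract intrinsic and outputs the resulting vector.
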MaterialFunction /Engine/Functions/Engine_MaterialFunctions01/Vectors/Refract
graph {
 properties {bExposeToLibrary=true, description=Calculates vector based Refraction using the hlsl intrinsic function.}
 nodes [
  MaterialExpressionFunctionInput_0 = FunctionInput() {description=The direction of the vector to refract., inputName=Ray Direction}
  MaterialExpressionFunctionInput_1 = FunctionInput() {
   bUsePreviewValueAsDefault=true
   description=The Normal of the interface between the two media.
   inputName=Surface Normal
   sortPriority=1
   previewValue=<0 0 1 1>
  }
  n1 = FunctionInput() {
   bUsePreviewValueAsDefault=true
   description=The refractive index of the media where the ray originates. The default is 1.0 for Air.
   inputName=Refractive Index Origin
   inputType=FunctionInput_Scalar
   sortPriority=2
   previewValue=<1 0 0 1>
  }
  n2 = FunctionInput() {
   bUsePreviewValueAsDefault=true
   description=The refractive index of the media the ray is intersecting. Defaults to 1.333 for Water.
   inputName=Refractive Index Target
   inputType=FunctionInput_Scalar
   sortPriority=3
   previewValue=<1.333 0 0 1>
  }
  MaterialExpressionDivide_0 = Divide(A=@n1, B=@n2)
  n3 = Custom!(Ray=@MaterialExpressionFunctionInput_0, Normal=@MaterialExpressionFunctionInput_1, Index=@MaterialExpressionDivide_0) {
   code="return refract(Ray, Normal, Index);"
   inputs [
    {inputName=Ray, input {expression=&self:MaterialExpressionFunctionInput_0}}
    {inputName=Normal, input {expression=&self:MaterialExpressionFunctionInput_1}}
    {inputName=Index, input {expression=&self:MaterialExpressionDivide_0}}
   ]
  }
  FunctionOutput(None=@n3) {outputName=Vector}
 ]
}
