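# RotateVector is a library material function that rotates VectorToRotate by the rotation that turns RestingVector (default +Z) toward LookAtVector.
# The cross product of RestingVector and normalized LookAtVector gives the axis, passed on without normalization.
# A Custom HLSL acos of their dot product gives the angle, converted to a 0-1 turn.
# RotateAboutAxis pivots on the origin, and its offset is added back to the input vector.
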MaterialFunction /Engine/Functions/Engine_MaterialFunctions01/Vectors/RotateVector
graph {
 properties {bExposeToLibrary=true, description=Rotates an input vector to point toward a given direction., libraryCategoriesText [Misc,Vectors]}
 nodes [
  CameraVectorWS()
  n1 = FunctionInput() {bUsePreviewValueAsDefault=true, description=the starting direction, inputName=RestingVector, sortPriority=2, previewValue=<0 0 1 1>}
  n2 = FunctionInput() {description=the direction to point to, inputName=LookAtVector, sortPriority=1, previewValue=<1 0 1 1>}
  n3 = Normalize(VectorInput=@n2)
  n4 = CrossProduct(A=@n1, B=@n3)
  MaterialExpressionDotProduct_4 = DotProduct(A=@n1, B=@n3)
  n5 = Custom!(a=@MaterialExpressionDotProduct_4) {
   code="return acos(a);"
   desc=aCos
   inputs [{inputName=a, input {expression=&self:MaterialExpressionDotProduct_4}}]
  }
  n6 = Divide(A=Multiply(A=@n5) {constB=57.29, desc=Radians to Degrees}) {constB=360, desc=Degrees to 0-1}
  n7 = Constant3Vector()
  n8 = FunctionInput() {inputName=VectorToRotate}
  FunctionOutput(None=Add(A=RotateAboutAxis(NormalizedRotationAxis=@n4, RotationAngle=@n6, PivotPoint=@n7, Position=@n8), B=@n8))
 ]
}
