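# Engine material function, exposed to the library under Vectors, that packs a 3D unit vector into a 2D octahedral encoding.
# A Vector input, expected to be pre-normalized, feeds a Custom HLSL node that calls the shader helper UnitVectorToOctahedron.
# The Float2 result is returned as the Octahedron output.
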MaterialFunction /Engine/Functions/Engine_MaterialFunctions01/Vectors/UnitVectorToOctahedron
graph {
 properties {bExposeToLibrary=true, description=Packs a v3 vector to a v2 Octahedron., libraryCategoriesText [Vectors]}
 nodes [
  MaterialExpressionFunctionInput_2 = FunctionInput() {bUsePreviewValueAsDefault=true, description=Make sure the input vector is normalized first., inputName=Vector, previewValue=<0 0 1 1>}
  n1 = Custom!(N=@MaterialExpressionFunctionInput_2) {
   code="return UnitVectorToOctahedron(N);"
   description=Vector to Octahedron
   outputType=CMOT_Float2
   inputs [{
     inputName=N
     input {expression=&self:MaterialExpressionFunctionInput_2}
    }]
  }
  FunctionOutput(None=@n1) {outputName=Octahedron}
 ]
}
